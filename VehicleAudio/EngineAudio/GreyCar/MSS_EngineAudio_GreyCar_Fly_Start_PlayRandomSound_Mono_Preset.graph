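# One-shot mono MetaSound Source preset for the GreyCar vehicle's fly-start engine sound.
# It inherits its graph from MSS_PlayRandomSound_Mono, supplying a single fly-start wave with narrow pitch and volume min/max ranges.
# Vehicle engine attenuation, sound class, low-spec concurrency sets and the Fortnite SFX submix govern playback.
# Modulation patches and buses drive its volume, pitch and high/low-pass filtering.
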MetaSoundSource /VehicleAudio/EngineAudio/GreyCar/MSS_EngineAudio_GreyCar_Fly_Start_PlayRandomSound_Mono_Preset
preset_parent=&MSS_PlayRandomSound_Mono
inputs [
 MaxPitch {type=Float}
 MaxVolume {type=Float}
 MinPitch {type=Float}
 MinVolume {type=Float}
 Pitch {type=Float}
 Seed {type=Int32}
 Shared State Behavior {type=Enum:SharedStateBehaviorType}
 Sounds {type=WaveAsset:Array}
 UE.Source.OnPlay {type=Trigger}
 Volume {type=Float}
]
interfaces [UE.OutputFormat.Mono:1.0, UE.Source.OneShot:1.0, UE.Source:1.1]
outputs [
 "UE.OutputFormat.Mono.Audio:0" {type=Audio}
 UE.Source.OneShot.OnFinished {type=Trigger}
 WaveDuration {type=Time}
]
pages [Default {input_defaults {
   MaxPitch=1.05
   MaxVolume=1.05
   MinPitch=0.95
   MinVolume=0.95
   Pitch=1
   Seed=-1
   Shared State Behavior=1
   UE.Source.OnPlay=null
   Volume=1
   Sounds [&/GreyCar_Fly_Start_01]
  }}]
properties {
 assetImportData=&self:AssetImportData
 attenuationSettings=&Vehicle_Engine_Att
 audioPropertiesSheet=&APS_Cosmetic_Vehicle_Engine
 maxDistance=12000
 priority=80
 sampleRate=48000
 soundClassObject=&Fort_Vehicles_Engine
 soundSubmixObject=&Fortnite_SFX
 concurrencySet [&Conc_Veh_EngineAudio_LowSpec, &Conc_Veh_EngineAudio_LowSpec_Actor]
 modulationSettings {
  highpassRouting=Union
  lowpassRouting=Union
  pitchRouting=Union
  volumeRouting=Union
  highpassModulationDestination {bEnableModulation=true, modulators [&PP_FortVehicles_Engine_HPF]}
  lowpassModulationDestination {bEnableModulation=true, modulators [&Patch_FortVehicles_Engine_LPF]}
  pitchModulationDestination {bEnableModulation=true, modulators [&DM_Bus_Vehicle_Engine_Pitch]}
  volumeModulationDestination {
   bEnableModulation=true
   modulators [&DM_VehicleEnginePatch, &Patch_FortVehicles_Engine_Vol]
  }
 }
}
