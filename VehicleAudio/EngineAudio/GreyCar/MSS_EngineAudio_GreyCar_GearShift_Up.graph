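# One-shot mono MetaSound source for the GreyCar engine's upshift gear change sound.
# On play, a random get picks one of the gearshift wave assets, a wave player plays it, and a decibel-to-linear gain input boosts the result.
# Playback end fires OnFinished, while engine attenuation, shifting concurrency sets and volume, pitch and filter modulation place it in the vehicle engine mix.
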MetaSoundSource /VehicleAudio/EngineAudio/GreyCar/MSS_EngineAudio_GreyCar_GearShift_Up
inputs [
 Decibels {type=Float}
 In Array {type=WaveAsset:Array}
 UE.Source.OnPlay {type=Trigger}
]
interfaces [UE.OutputFormat.Mono:1.0, UE.Source.OneShot:1.0, UE.Source:1.1]
outputs [
 "UE.OutputFormat.Mono.Audio:0" {type=Audio}
 UE.Source.OneShot.OnFinished {type=Trigger}
]
pages [Default {
  input_defaults {
   Decibels=3.5
   UE.Source.OnPlay=null
   In Array [
    &/EngineAudio_GreyCar_Gearshift_01
    &/EngineAudio_GreyCar_Gearshift_02
    &/EngineAudio_GreyCar_Gearshift_03
   ]
  }
  nodes [
   DecibelstoLinearGain_0 = UE.Decibels to Linear Gain:Float(Decibels=^Decibels)
   ArrayRandomGet_0 = Array.Random Get:WaveAsset:Array(In Array=^In Array, Next=^UE.Source.OnPlay)
   WavePlayer_0 = UE.Wave Player:Mono(Play=@ArrayRandomGet_0.On Next, Wave Asset=@ArrayRandomGet_0.Value) {also {version [1,1]}}
   Multiply_0 = UE.Multiply:Audio by Float(AdditionalOperands=@DecibelstoLinearGain_0.Linear Gain, PrimaryOperand=@WavePlayer_0.Out Mono)
  ]
  links [{dst="@UE.OutputFormat.Mono.Audio:0", src=Multiply_0.Out}, {dst="@UE.Source.OneShot.OnFinished", src=WavePlayer_0.On Finished}]
 }]
properties {
 assetImportData=&self:AssetImportData
 attenuationSettings=&Vehicle_Engine_Att
 audioPropertiesSheet=/VehicleAudio/PropertySheets/APS_Cosmetic_Vehicle_Engine_Shifting
 maxDistance=12000
 sampleRate=48000
 soundClassObject=&Fort_Vehicles_Engine
 concurrencySet [
  /VehicleAudio/EngineAudio/Concurrency/CONC_Veh_EngineAudio_Shitfing
  /VehicleAudio/EngineAudio/Concurrency/CONC_Veh_EngineAudio_Shifting_Actor
 ]
 modulationSettings {
  highpassRouting=Union
  lowpassRouting=Union
  pitchRouting=Union
  volumeRouting=Union
  highpassModulationDestination {bEnableModulation=true, modulators [&PP_FortVehicles_Engine_HPF]}
  lowpassModulationDestination {bEnableModulation=true, modulators [&Patch_FortVehicles_Engine_LPF]}
  pitchModulationDestination {bEnableModulation=true, modulators [&DM_Bus_Vehicle_Engine_Pitch]}
  volumeModulationDestination {
   bEnableModulation=true
   modulators [&DM_VehicleEnginePatch, &Patch_FortVehicles_Engine_Vol]
  }
 }
}
