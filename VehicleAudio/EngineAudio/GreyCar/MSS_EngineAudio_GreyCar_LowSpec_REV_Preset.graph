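# Looping mono MetaSound source for GreyCar's low-spec engine audio, wrapping the shared LowSpec REV parent patch with car-specific assets.
# It adds an in-air loop, driven by an ADSR latched from grounded state and BonusSpeed, and a bonus-speed loop gated by a crossfade segment on BonusSpeed.
# Both layer levels combine into an interpolated dB offset on REV volume, rebalancing the engine as the loops mix in.
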MetaSoundSource /VehicleAudio/EngineAudio/GreyCar/MSS_EngineAudio_GreyCar_LowSpec_REV_Preset
inputs [
 Blowoff.Sound {type=WaveAsset:Array, access=Constructor}
 Blowoff.Volume {type=Float, access=Constructor}
 BonusSpeedLayerVolume {type=Float}
 BonusSpeedLoop {type=WaveAsset, access=Constructor}
 BulletWhip.OnStop {type=Trigger}
 BulletWhip.PassByAlignment {type=Float}
 BulletWhip.PassByAzimuth {type=Float}
 BulletWhip.PassByDistance {type=Float}
 Character.Affiliation {type=Enum:FortAudioAffiliation}
 Character.Health {type=Float}
 Character.UniqueID {type=Int32}
 Downshift {type=Trigger}
 EngineBlowoff {type=Trigger}
 Exhaust.Pops.Blowoff.Volume {type=Float, access=Constructor}
 Exhaust.Pops.Long {type=WaveAsset:Array, access=Constructor}
 Exhaust.Pops.Pitch {type=Float, access=Constructor}
 Exhaust.Pops.Shift.Volume {type=Float, access=Constructor}
 Exhaust.Pops.Short {type=WaveAsset:Array, access=Constructor}
 Exhaust.Pops.Volume {type=Float, access=Constructor}
 Exhaust.Throttle.EQ.FilterQ {type=Float, access=Constructor}
 Exhaust.Throttle.EQ.FilterType {type=Enum:BiquadFilterType, access=Constructor}
 Exhaust.Throttle.EQ.OffFreq {type=Float, access=Constructor}
 Exhaust.Throttle.EQ.OffGain {type=Float, access=Constructor}
 Exhaust.Throttle.EQ.On Throttle Frequency {type=Float, access=Constructor}
 Exhaust.Throttle.EQ.OnGain {type=Float, access=Constructor}
 Exhaust.Throttle.InterpTime {type=Time, access=Constructor}
 Exhaust.Throttle.Volume.Off {type=Float, access=Constructor}
 Exhaust.Throttle.Volume.On {type=Float, access=Constructor}
 Fort.Vehicle.Acceleration {type=Float}
 Fort.Vehicle.BlowoffStrength {type=Float}
 Fort.Vehicle.BonusSpeed {type=Float}
 Fort.Vehicle.EnableShifting {type=Bool}
 Fort.Vehicle.Gear {type=Int32}
 Fort.Vehicle.IsGrounded {type=Bool}
 Fort.Vehicle.Load {type=Float}
 Fort.Vehicle.LowFuel {type=Bool}
 Fort.Vehicle.NormalSpeed {type=Float}
 Fort.Vehicle.OnBlowoff {type=Trigger}
 Fort.Vehicle.OnDownshift {type=Trigger}
 Fort.Vehicle.OnThrottleEngaged {type=Trigger}
 Fort.Vehicle.OnThrottleReleased {type=Trigger}
 Fort.Vehicle.OnUpshift {type=Trigger}
 Fort.Vehicle.Pitch {type=Float}
 Fort.Vehicle.RPM {type=Float}
 Fort.Vehicle.Throttle {type=Float}
 GearChangeTime {type=Time, access=Constructor}
 InAirAddLayerVolume {type=Float}
 InAirLoop {type=WaveAsset}
 IsLocal {type=Bool}
 Mono {type=Rev Asset, access=Constructor}
 REV.EnableShifting {type=Bool}
 REV.Gear {type=Int32}
 REV.RPM {type=Float}
 REV.Throttle {type=Float}
 REV.Velocity {type=Float}
 REVVolume {type=Float}
 Settings {type=Rev Settings, access=Constructor}
 Speed_Normalized {type=Float}
 UE.Source.OnPlay {type=Trigger}
 UE.Source.Orientation.Azimuth {type=Float}
 UE.Source.Orientation.Elevation {type=Float}
 Upshift {type=Trigger}
 Value {type=Float}
 Value 1 {type=Float}
]
interfaces [Fort.BulletWhip:1.0, Fort.Character:1.0, Fort.Vehicle:1.0, REV:1.0, UE.OutputFormat.Mono:1.0, UE.Source.Orientation:1.0, UE.Source:1.1]
outputs [
 Downshift {type=Trigger}
 Fundamental Frequency {type=Float}
 Gear {type=Int32}
 RPM {type=Float}
 SpeedNormalized {type=Float}
 Throttle {type=Float}
 "UE.OutputFormat.Mono.Audio:0" {type=Audio}
 "UE.OutputFormat.Stereo.Audio:1" {type=Audio}
 Upshift {type=Trigger}
]
pages [Default {
  input_defaults {
   Blowoff.Volume=0
   BonusSpeedLayerVolume=6
   BonusSpeedLoop=&/EngineAudio_GreyCar_Sparking_Loop_01
   BulletWhip.OnStop=null
   BulletWhip.PassByAlignment=0
   BulletWhip.PassByAzimuth=-90
   BulletWhip.PassByDistance=1
   Character.Affiliation=0
   Character.Health=0
   Character.UniqueID=0
   Downshift=null
   EngineBlowoff=null
   Exhaust.Pops.Blowoff.Volume=0
   Exhaust.Pops.Pitch=0
   Exhaust.Pops.Shift.Volume=0
   Exhaust.Pops.Volume=0
   Exhaust.Throttle.EQ.FilterQ=1
   Exhaust.Throttle.EQ.FilterType=4
   Exhaust.Throttle.EQ.OffFreq=20000
   Exhaust.Throttle.EQ.OffGain=0
   Exhaust.Throttle.EQ.On Throttle Frequency=20000
   Exhaust.Throttle.EQ.OnGain=0
   Exhaust.Throttle.InterpTime=0.1
   Exhaust.Throttle.Volume.Off=0
   Exhaust.Throttle.Volume.On=0
   Fort.Vehicle.Acceleration=0
   Fort.Vehicle.BlowoffStrength=0
   Fort.Vehicle.BonusSpeed=0
   Fort.Vehicle.EnableShifting="<MetaSoundBuilderResult.SUCCEEDED: 0>"
   Fort.Vehicle.Gear=0
   Fort.Vehicle.IsGrounded=null
   Fort.Vehicle.Load=0
   Fort.Vehicle.LowFuel=null
   Fort.Vehicle.NormalSpeed=0
   Fort.Vehicle.OnBlowoff=null
   Fort.Vehicle.OnDownshift=null
   Fort.Vehicle.OnThrottleEngaged=null
   Fort.Vehicle.OnThrottleReleased=null
   Fort.Vehicle.OnUpshift=null
   Fort.Vehicle.Pitch=0
   Fort.Vehicle.RPM=0
   Fort.Vehicle.Throttle=0
   GearChangeTime=0.1
   InAirAddLayerVolume=-2.592
   InAirLoop=&/GreyCar_Fly_Loop_01
   IsLocal=null
   Mono=&/EngineAudio_GreyCar
   REV.EnableShifting="<MetaSoundBuilderResult.SUCCEEDED: 0>"
   REV.Gear=0
   REV.RPM=0
   REV.Throttle=0
   REV.Velocity=0
   REVVolume=6
   Settings=&/REVSettings_GreyCar
   Speed_Normalized=0
   UE.Source.OnPlay=null
   UE.Source.Orientation.Azimuth=0
   UE.Source.Orientation.Elevation=0
   Upshift=null
   Value=0
   Value 1=0
   Blowoff.Sound [
    &/Engine_GreyCar_Blowoff_Techy_01
    &/Engine_GreyCar_Blowoff_Techy_02
    &/Engine_GreyCar_Blowoff_Techy_03
   ]
   Exhaust.Pops.Long []
   Exhaust.Pops.Short []
  }
  nodes [
   n1 = Fortnite.CrossfadeSegmentManager:Float(In=^Fort.Vehicle.BonusSpeed) {also {defaults {Max=1.1, MaxRange=0, Min=0.1}}}
   n2 = InitVariable.Float()
   TriggerOnValueChange_0 = UE.Trigger On Value Change:Bool(Value=^Fort.Vehicle.IsGrounded)
   TriggerOnValueChange_1 = UE.Trigger On Value Change:Float(Value=^Fort.Vehicle.BonusSpeed)
   TriggerAnyTriggerAny3_0 = "TriggerAny.Trigger Any (3)"(In 2=^UE.Source.OnPlay, In 0=@TriggerOnValueChange_0.Trigger, In 1=@TriggerOnValueChange_1.Trigger)
   n3 = TriggerCompare.Clamp:Bool(A=^Fort.Vehicle.IsGrounded, Compare=@TriggerAnyTriggerAny3_0.Out)
   n4 = TriggerCompare.Clamp:Float(A=^Fort.Vehicle.BonusSpeed, Compare=@n3.True) {also {defaults {Type=3}}}
   n5 = VariableDeferredAccessor.Trigger(Variable=InitVariable.Trigger().Variable)
   TriggerOnce_0 = UE.Trigger Once(Trigger In=@n4.True, Reset=@n5.Value)
   n6 = VariableDeferredAccessor.Trigger(Variable=InitVariable.Trigger().Variable)
   n7 = VariableAccessor.Trigger(Variable=VariableMutator.Trigger(Value=@TriggerOnce_0.Trigger Out, Variable=@n6.Variable).Variable)
   TriggerAnyTriggerAny2_0 = "TriggerAny.Trigger Any (2)"(In 1=TriggerCompare.Clamp:Float(A=^Fort.Vehicle.BonusSpeed, Compare=@n3.False) {also {defaults {B=0, Type=3}}}.True, In 0=@n4.False)
   TriggerOnce_1 = UE.Trigger Once(Trigger In=@TriggerAnyTriggerAny2_0.Out, Reset=@n6.Value)
   ADSREnvelopeADSREnvelope_0 = ADSR Envelope.ADSR Envelope:Float(
    Trigger Attack=VariableAccessor.Trigger(Variable=@n7.Variable).Value
    Trigger Release=VariableAccessor.Trigger(Variable=VariableMutator.Trigger(Value=@TriggerOnce_1.Trigger Out, Variable=@n5.Variable).Variable).Value
   ) {also {defaults {Attack Time=1, Sustain Level=4}}}
   n8 = InitVariable.Float()
   n9 = UE.Add:Float(
    PrimaryOperand=VariableAccessor.Float(Variable=VariableMutator.Float(Value=@n1.Volume, Variable=@n2.Variable).Variable).Value
    AdditionalOperands=VariableAccessor.Float(Variable=VariableMutator.Float(Value=@ADSREnvelopeADSREnvelope_0.Out Envelope, Variable=@n8.Variable).Variable).Value
   )
   DecibelstoLinearGain_0 = UE.Decibels to Linear Gain:Float(Decibels=MapRange.MapRange:Float(In=UE.Multiply:Float(PrimaryOperand=@n9.Out) {also {defaults {AdditionalOperands=0.5}}}.Out) {also {defaults {Out Range B=-8}}}.Out Value)
   MSS_EngineAudio_LowSpec_Parent_REV_0 = /VehicleAudio/Patches/MSS_EngineAudio_LowSpec_Parent_REV(
    Blowoff.Sound=^Blowoff.Sound
    Blowoff.Volume=^Blowoff.Volume
    Character.Affiliation=^Character.Affiliation
    Character.Health=^Character.Health
    Character.UniqueID=^Character.UniqueID
    Downshift=^Downshift
    EngineBlowoff=^EngineBlowoff
    Exhaust.Pops.Blowoff.Volume=^Exhaust.Pops.Blowoff.Volume
    Exhaust.Pops.Long=^Exhaust.Pops.Long
    Exhaust.Pops.Pitch=^Exhaust.Pops.Pitch
    Exhaust.Pops.Shift.Volume=^Exhaust.Pops.Shift.Volume
    Exhaust.Pops.Short=^Exhaust.Pops.Short
    Exhaust.Pops.Volume=^Exhaust.Pops.Volume
    Exhaust.Throttle.EQ.FilterQ=^Exhaust.Throttle.EQ.FilterQ
    Exhaust.Throttle.EQ.FilterType=^Exhaust.Throttle.EQ.FilterType
    Exhaust.Throttle.EQ.OffFreq=^Exhaust.Throttle.EQ.OffFreq
    Exhaust.Throttle.EQ.OffGain=^Exhaust.Throttle.EQ.OffGain
    Exhaust.Throttle.EQ.On Throttle Frequency=^Exhaust.Throttle.EQ.On Throttle Frequency
    Exhaust.Throttle.EQ.OnGain=^Exhaust.Throttle.EQ.OnGain
    Exhaust.Throttle.InterpTime=^Exhaust.Throttle.InterpTime
    Exhaust.Throttle.Volume.Off=^Exhaust.Throttle.Volume.Off
    Exhaust.Throttle.Volume.On=^Exhaust.Throttle.Volume.On
    Fort.Vehicle.Acceleration=^Fort.Vehicle.Acceleration
    Fort.Vehicle.BlowoffStrength=^Fort.Vehicle.BlowoffStrength
    Fort.Vehicle.BonusSpeed=^Fort.Vehicle.BonusSpeed
    Fort.Vehicle.EnableShifting=^Fort.Vehicle.EnableShifting
    Fort.Vehicle.Gear=^Fort.Vehicle.Gear
    Fort.Vehicle.IsGrounded=^Fort.Vehicle.IsGrounded
    Fort.Vehicle.Load=^Fort.Vehicle.Load
    Fort.Vehicle.NormalSpeed=^Fort.Vehicle.NormalSpeed
    Fort.Vehicle.OnBlowoff=^Fort.Vehicle.OnBlowoff
    Fort.Vehicle.OnDownshift=^Fort.Vehicle.OnDownshift
    Fort.Vehicle.OnThrottleEngaged=^Fort.Vehicle.OnThrottleEngaged
    Fort.Vehicle.OnThrottleReleased=^Fort.Vehicle.OnThrottleReleased
    Fort.Vehicle.OnUpshift=^Fort.Vehicle.OnUpshift
    Fort.Vehicle.Pitch=^Fort.Vehicle.Pitch
    Fort.Vehicle.RPM=^Fort.Vehicle.RPM
    Fort.Vehicle.Throttle=^Fort.Vehicle.Throttle
    GearChangeTime=^GearChangeTime
    IsLocal=^IsLocal
    Mono=^Mono
    REV.EnableShifting=^REV.EnableShifting
    REV.Gear=^REV.Gear
    REV.RPM=^REV.RPM
    REV.Throttle=^REV.Throttle
    REV.Velocity=^REV.Velocity
    Settings=^Settings
    Speed_Normalized=^Speed_Normalized
    UE.Source.OnPlay=^UE.Source.OnPlay
    UE.Source.Orientation.Azimuth=^UE.Source.Orientation.Azimuth
    UE.Source.Orientation.Elevation=^UE.Source.Orientation.Elevation
    Upshift=^Upshift
    REVVolume=UE.Multiply:Float(
     PrimaryOperand=^REVVolume
     AdditionalOperands=UE.InterpTo:Audio(Target=@DecibelstoLinearGain_0.Linear Gain) {also {defaults {Interp Time=0.2}}}.Value
    ).Out
   )
   DecibelstoLinearGain_2 = UE.Decibels to Linear Gain:Float(Decibels=^InAirAddLayerVolume)
   DecibelstoLinearGain_1 = UE.Decibels to Linear Gain:Float(Decibels=^BonusSpeedLayerVolume)
   WavePlayer_0 = UE.Wave Player:Mono(Wave Asset=^InAirLoop, Stop=@ADSREnvelopeADSREnvelope_0.On Done, Play=@n7.Value) {also {defaults {Loop="<MetaSoundBuilderResult.SUCCEEDED: 0>"}, version [1,1]}}
   WavePlayer_1 = UE.Wave Player:Mono(Wave Asset=^BonusSpeedLoop, Play=@n1.Activate, Stop=@n1.Deactivate) {also {defaults {Loop="<MetaSoundBuilderResult.SUCCEEDED: 0>"}, version [1,1]}}
   AudioMixerAudioMixerMono3_0 = "AudioMixer.Audio Mixer (Mono, 3)"(
    In 2=@MSS_EngineAudio_LowSpec_Parent_REV_0."UE.OutputFormat.Mono.Audio:0"
    Gain 0=UE.Multiply:Float(AdditionalOperands=@ADSREnvelopeADSREnvelope_0.Out Envelope, PrimaryOperand=@DecibelstoLinearGain_2.Linear Gain).Out
    Gain 1=UE.Multiply:Float(PrimaryOperand=@DecibelstoLinearGain_1.Linear Gain, AdditionalOperands=@n1.Volume).Out
    In 0=@WavePlayer_0.Out Mono
    In 1=@WavePlayer_1.Out Mono
   )
  ]
  links [
   {dst="@UE.OutputFormat.Mono.Audio:0", src=AudioMixerAudioMixerMono3_0.Out}
   {dst="@UE.OutputFormat.Stereo.Audio:1", src=AudioMixerAudioMixerMono3_0.Out}
   {dst="@Downshift", src=MSS_EngineAudio_LowSpec_Parent_REV_0.Downshift}
   {dst="@Fundamental Frequency", src=MSS_EngineAudio_LowSpec_Parent_REV_0.Fundamental Frequency}
   {dst="@Gear", src=MSS_EngineAudio_LowSpec_Parent_REV_0.Gear}
   {dst="@RPM", src=MSS_EngineAudio_LowSpec_Parent_REV_0.RPM}
   {dst="@SpeedNormalized", src=MSS_EngineAudio_LowSpec_Parent_REV_0.SpeedNormalized}
   {dst="@Throttle", src=MSS_EngineAudio_LowSpec_Parent_REV_0.Throttle}
   {dst="@Upshift", src=MSS_EngineAudio_LowSpec_Parent_REV_0.Upshift}
  ]
 }]
properties {
 assetImportData=&self:AssetImportData
 attenuationSettings=&Vehicle_Engine_Att
 audioPropertiesSheet=&APS_Cosmetic_Vehicle_Engine
 bLooping=true
 duration=10000
 maxDistance=12000
 priority=80
 qualitySetting=Vehicles_REV_Mono
 sampleRate=48000
 soundClassObject=&Fort_Vehicles_Engine
 soundSubmixObject=&Fortnite_SFX
 volume=1.3
 concurrencySet [&Conc_Veh_EngineAudio_LowSpec, &Conc_Veh_EngineAudio_LowSpec_Actor]
 modulationSettings {
  highpassRouting=Union
  lowpassRouting=Union
  pitchRouting=Union
  volumeRouting=Union
  highpassModulationDestination {bEnableModulation=true, modulators [&PP_FortVehicles_Engine_HPF]}
  lowpassModulationDestination {bEnableModulation=true, modulators [&Patch_FortVehicles_Engine_LPF]}
  pitchModulationDestination {bEnableModulation=true, modulators [&DM_Bus_Vehicle_Engine_Pitch]}
  volumeModulationDestination {
   bEnableModulation=true
   modulators [&DM_VehicleEnginePatch, &Patch_FortVehicles_Engine_Vol]
  }
 }
}
